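# Animation Blueprint for the large male ZebraScramble Bacon cosmetic body, which follows the pose of its attached parent mesh.
# Copy Pose From Mesh feeds four chained Transform Bone nodes that add fixed, asymmetric rotation offsets to both thighs and feet.
# The adjusted pose then runs through the body's own Control Rig, which receives the emote-playing flag as an input.
# The event graph holds only an empty update event.
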AnimBlueprint /BRCosmetics/Characters/Player/Male/Large/Bodies/M_LRG_ZebraScramble_Bacon/Meshes/M_LRG_ZebraScramble_Bacon_AnimBP
target_skeleton=&/M_LRG_ZebraScramble_Bacon_Skeleton
graphs [
 AnimGraph {nodes [
   n1 = "Transform (Modify) Bone - Bone: thigh_l"!(
    ComponentPose=Local To Component(LocalPose=Copy Pose From Mesh() {bUseAttachedParent=true})
    bAlphaBoolEnabled:bool=True
    Translation="0,0,0"
    Rotation="0,0,-15.5"
    Scale="1,1,1"
   ) -> (Pose) {
    alphaInputType=Bool
    rotationMode=BMM_Additive
    translationMode=BMM_Additive
    alphaBoolBlend {blendInTime=0.3, blendOption=Linear, blendOutTime=0.2}
    boneToModify {boneName=thigh_l}
    rotation {roll=-15.5}
   }
   n2 = "Transform (Modify) Bone - Bone: thigh_r"!(ComponentPose=@n1, bAlphaBoolEnabled:bool=True, Translation="0,0,0", Rotation="0,0,-20", Scale="1,1,1") -> (Pose) {
    alphaInputType=Bool
    rotationMode=BMM_Additive
    alphaBoolBlend {blendInTime=0.3, blendOption=Linear, blendOutTime=0.2}
    boneToModify {boneName=thigh_r}
    rotation {roll=-20}
   }
   n3 = "Transform (Modify) Bone - Bone: foot_l"!(ComponentPose=@n2, bAlphaBoolEnabled:bool=True, Translation="0,0,0", Rotation="-0.000,0.000,15.304", Scale="1,1,1") -> (Pose) {
    alphaInputType=Bool
    rotationMode=BMM_Additive
    translationMode=BMM_Additive
    alphaBoolBlend {blendInTime=0.3, blendOption=Linear, blendOutTime=0.2}
    boneToModify {boneName=foot_l}
    rotation {roll=15.304}
   }
   n4 = "Transform (Modify) Bone - Bone: foot_r"!(ComponentPose=@n3, bAlphaBoolEnabled:bool=True, Translation="0,0,0", Rotation="17.250,6.686,31.957", Scale="1,1,1") -> (Pose) {
    alphaInputType=Bool
    rotationMode=BMM_Additive
    alphaBoolBlend {blendInTime=0.3, blendOption=Linear, blendOutTime=0.2}
    boneToModify {boneName=foot_r}
    rotation=<rot 17.25 6.686 31.957>
   }
   n5 = Control Rig!(Source=Component To Local(ComponentPose=@n4), Alpha, Is Emoting:bool=Get bIsPlayingEmote()) -> (Pose) {
    alphaBoolBlend {blendOption=Linear}
    controlRigAssetReference {blueprintRigClass=&/M_LRG_ZebraScramble_Bacon_CtrlRig.M_LRG_ZebraScramble_Bacon_CtrlRig_C}
   }
   Output Pose(Result=@n5) {layerGroup=None}
  ]}
 EventGraph {nodes [Event BlueprintUpdateAnimation()]}
]
variables []
